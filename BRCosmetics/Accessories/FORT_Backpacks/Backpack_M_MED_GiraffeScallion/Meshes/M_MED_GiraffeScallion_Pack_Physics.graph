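# Physics asset for the GiraffeScallion backpack cosmetic, providing collision and joints for two lock bones, left and right, that hang off the pack.
# Each of the three bodies (pack, dyn_lock_l, dyn_lock_r) is a single sphere, with the lock spheres offset from their bones.
# Two mirrored constraints join each lock to the pack, locking twist and swing1 so each lock rotates hinge-like on swing2 alone.
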PhysicsAsset /BRCosmetics/Accessories/FORT_Backpacks/Backpack_M_MED_GiraffeScallion/Meshes/M_MED_GiraffeScallion_Pack_Physics
bodies {
 dyn_lock_l [{shapeType=Sphere, radius=5, shapeName=dyn_lock_l_sphere, center=<2.415 0 -4.0499>}]
 dyn_lock_r [{shapeType=Sphere, radius=5, shapeName=dyn_lock_r_sphere, center=<2.415 0 -4.0499>}]
 pack [{shapeType=Sphere, radius=5}]
}
constraints [
 {
  bone1Name=dyn_lock_l
  bone2Name=pack
  swing1LimitDegrees=0
  swing1Motion=Locked
  swing2LimitDegrees=70
  twistLimitDegrees=0
  twistMotion=Locked
 }
 {
  bone1Name=dyn_lock_r
  bone2Name=pack
  swing1LimitDegrees=0
  swing1Motion=Locked
  swing2LimitDegrees=70
  twistLimitDegrees=0
  twistMotion=Locked
 }
]
