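# Anim Blueprint for the GiraffeScallion backpack cosmetic, adding secondary physics motion to its skeleton.
# A cached input pose feeds two branches: a RigidBody node using the pack's physics asset, and chained AnimDynamics nodes on the left and right lock bones.
# bRigidBodyEnabled selects RigidBody when true and AnimDynamics otherwise.
# Both branches override gravity, and a Sim Space Settings variable drives the RigidBody simulation space.
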AnimBlueprint /BRCosmetics/Accessories/FORT_Backpacks/Backpack_M_MED_GiraffeScallion/Meshes/M_MED_GiraffeScallion_Pack_AnimBP
target_skeleton=&/M_MED_GiraffeScallion_Pack_Skeleton
graphs [
 AnimGraph {nodes [
   n1 = Anim Dynamics - Bone: dyn_lock_l(ComponentPose=Local To Component(LocalPose=Use cached pose 'Input Cache'()), Alpha=1, ExternalForce="0,0,0") {
    angularBiasOverride=1
    angularDampingOverride=1
    bOverrideAngularBias=true
    bOverrideAngularDamping=true
    bUseGravityOverride=true
    bUsePlanarLimit=false
    simulationSpace=World
    alphaBoolBlend {blendOption=Linear}
    boundBone {boneName=dyn_lock_l}
    componentAppliedLinearAccClamp=<3000 3000 3000>
    componentLinearAccScale=<0.3 0.5 0.3>
    componentLinearVelScale=<0.3 0.5 0.3>
    gravityOverride {z=-100}
    physicsBodyDefinitions [{
      boundBone {boneName=dyn_lock_l}
      boxExtents=<12 12 12>
      constraintSetup {angularLimitsMax=<0 20 0>, angularLimitsMin=<0 -120 0>, angularTarget=<1 0 0>}
      localJointOffset=<0 0 -3>
     }]
    relativeSpaceBone {boneName=root}
   }
   n2 = Anim Dynamics - Bone: dyn_lock_r(ComponentPose=@n1, Alpha=1, ExternalForce="0,0,0") {
    angularBiasOverride=1
    angularDampingOverride=1
    bOverrideAngularBias=true
    bOverrideAngularDamping=true
    bUseGravityOverride=true
    bUsePlanarLimit=false
    simulationSpace=World
    alphaBoolBlend {blendOption=Linear}
    boundBone {boneName=dyn_lock_r}
    componentAppliedLinearAccClamp=<3000 3000 3000>
    componentLinearAccScale=<0.3 0.5 0.3>
    componentLinearVelScale=<0.3 0.5 0.3>
    gravityOverride {z=-100}
    physicsBodyDefinitions [{
      boundBone {boneName=dyn_lock_r}
      boxExtents=<12 12 12>
      constraintSetup {angularLimitsMax=<0 20 0>, angularLimitsMin=<0 -120 0>, angularTarget=<1 0 0>}
      localJointOffset=<0 0 -3>
     }]
    relativeSpaceBone {boneName=root}
   }
   n3 = RigidBody!(
    ComponentPose=Local To Component(LocalPose=Use cached pose 'Input Cache'())
    Alpha
    ExternalForce
    SimSpaceSettings:Sim Space Settings Structure=Get Sim Space Settings() -> (Sim Space Settings:Sim Space Settings Structure)
   ) -> (Pose) {
    bOverrideWorldGravity=true
    overridePhysicsAsset=&/M_MED_GiraffeScallion_Pack_Physics
    simulationSpace=BaseBoneSpace
    alphaBoolBlend {blendOption=Linear}
    baseBoneRef {boneName=root}
    overrideWorldGravity {z=-1500}
    simSpaceSettings {
     dampingAlpha=10
     maxAngularAcceleration=3000
     maxAngularVelocity=20
     maxLinearAcceleration=2500
     maxLinearVelocity=700
     velocityScaleZ=2
     worldAlpha=1
     externalLinearDragV=<0.5 0.5 0.5>
    }
   }
   Output Pose(Result=Blend Poses by bool(BlendPose_0=Component To Local(ComponentPose=@n3), BlendPose_1=Component To Local(ComponentPose=@n2), bActiveValue=Get bRigidBodyEnabled()) {blendType=Linear}) {layerGroup=None}
   Save cached pose 'Input Cache'(Pose=Input Pose()) {cacheName=Input Cache}
  ]}
 EventGraph {nodes [
   TryGetPawnOwner()
   Event BlueprintUpdateAnimation()
  ]}
]
variables [Sim Space Settings {
  type=SimSpaceSettings
  default {
   dampingAlpha=0
   maxAngularAcceleration=2000
   maxAngularVelocity=30
   maxLinearAcceleration=2001
   maxLinearVelocity=700
   velocityScaleZ=2
   worldAlpha=1
   externalLinearDragV=<1 1 1>
  }
 }]
